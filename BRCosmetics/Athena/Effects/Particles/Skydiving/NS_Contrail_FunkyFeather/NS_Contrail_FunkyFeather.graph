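# Niagara skydiving contrail cosmetic that trails mesh particles, speedlines and wind ribbons from the character.
# Bills and Ducks spawn meshes at skeletal bones, shaped by drag, gravity and curl noise, with paraglide lean driving extra spawning and lift.
# WindTrails rebuilds four bone-attached ribbons every update from persistent particles offset along contrail velocity, while Speedlines stretches torus-spawned sprites by speed.
# Front-end modifiers swap offsets and forces for preview.
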NiagaraSystem /BRCosmetics/Athena/Effects/Particles/Skydiving/NS_Contrail_FunkyFeather/NS_Contrail_FunkyFeather
dependencies {
 usedDataInterfaces [
  &NiagaraDataInterfaceArrayColor
  &NiagaraDataInterfaceMeshRendererInfo
  &NiagaraDataInterfaceVectorCurve
  &NiagaraDataInterfaceCurve
  &NiagaraDataInterfaceCamera
  &NiagaraDataInterfaceColorCurve
  &NiagaraDataInterfaceVector2DCurve
  &NiagaraDataInterfaceArrayFloat
  &NiagaraDataInterfaceSkeletalMesh
 ]
 usedDynamicInputs [
  &Multiply_Vector2D_ByFloat
  &SetBoolByFloatComparison
  &Multiply_Vector
  &RotateVectorByQuaternion
  /Niagara/DynamicInputs/Rotation/InvertQuaternion
 ]
 usedModules [
  &self:SystemSpawnScript.SystemScriptSource.SystemScriptGraph.NiagaraNodeAssignment_2.SetVariables_#1
  &SystemState
  /Game/Effects/Niagara/Modules/Skydiving/FN_DivingAttributes/v5/FN_DivingAttributes
  /Game/Effects/Niagara/Modules/Contrails/FN_Contrail_Setup
  &self:SystemSpawnScript.SystemScriptSource.SystemScriptGraph.NiagaraNodeAssignment_4.SetVariables_#2
  &FN_Integrator
  &self:SystemSpawnScript.SystemScriptSource.SystemScriptGraph.NiagaraNodeAssignment_0.SetVariables_#3
 ]
}
emitters {
 Bills {
  emitterUpdateScript [
   &EmitterState(
    Inactive Response=::NewEnumerator1
    Life Cycle Mode=::NewEnumerator1
    Loop Behavior=::NewEnumerator1
    Loop Duration Mode=::NewEnumerator1
    UseLoopDelay=true
   )
   &SpawnRate(SpawnRate=&Lerp_Float(A=60, B=60, Alpha=$User.ParaGlideLeanAlpha), Use Spawn Probability=false)
   &SpawnRate(SpawnGroup=1, SpawnRate=&Lerp_Float(A=0, B=75, Alpha=$User.ParaGlideLeanAlpha), Use Spawn Probability=false) {module=SpawnRate001}
  ]
  particleSpawnScript [
   &PartitionParticles(Partition Mode:ENiagara_PartitionMode_v2=::NewEnumerator5)
   &InitializeParticle(
    Color=&SelectLinearColorFromArray(
     Array Sampling Mode=::NewEnumerator0
     Direct Array Index=?
     Color Selection Array=DataInterface() {propertyValues {ColorData [<rgba 0.36 0.9 0.5535 1>, <rgba 0.8019 0.5 1 1>, <rgba 1 0.5127 0.3 1>]}}
     Random Seed=0
    )
    Lifetime=&RandomRangeFloat(Minimum=0.5, Maximum=1.5)
    Mesh Renderer Array Visibility Mode=::NewEnumerator3
    Mesh Renderer Info=DataInterface() {propertyValues {MeshRenderer=&self:Bills_0.NiagaraMeshRendererProperties_1}}
    Mesh Scale Mode=::NewEnumerator3
    Sprite Rotation Mode=::NewEnumerator2
    Sprite Size Mode=::NewEnumerator3
    Uniform Sprite Size=&RandomRangeFloat(Minimum=3, Maximum=5)
    UsePositionOffset=true
    Write Material Random=true
   )
   &SkeletalMeshLocation(
    Bone / Socket Index=$Particles.PartitionParticles.Partition
    Bone Sampling Mode=::NewEnumerator4
    MaskSpawnGroup=true
    Orientation Sampling=::NewEnumerator1
    Position Sampling=::NewEnumerator1
    Sampled Position Offset=&FN_FrontEndVector3Modifier(Gameplay=<0 0 -5>, Front End=<0 0 0>)
    Sampled Position Offset Coordinate Space:ENiagaraSkelMeshTransforms=::NewEnumerator1
    Skeletal Mesh=$System.Skeletal Mesh
   )
   &SkeletalMeshLocation(
    Bone / Socket Index=4
    Bone Sampling Mode=::NewEnumerator4
    MaskSpawnGroup=true
    Orientation Sampling=::NewEnumerator1
    Position Sampling=::NewEnumerator1
    Sampled Position Offset=&FN_FrontEndVector3Modifier(Gameplay=<0 -85 -200>, Front End=<0 85 -200>)
    Sampled Position Offset Coordinate Space:ENiagaraSkelMeshTransforms=::NewEnumerator2
    Skeletal Mesh=$System.Skeletal Mesh
    Spawn Group Mask=1
   ) {module=SkeletalMeshLocation001}
   &InitialMeshOrientation(
    Mesh Orientation Mode=::NewEnumerator6
    Orientation Quaternion:Quat4f=$Particles.SkeletalMeshLocation.SampledRotation
    Rotation=&MergeHelper_VectorSwitch(
     0=&RandomRangeVector(Minimum=<-0.1 -0.1 0>, Maximum=<0.1 0.1 1>)
     1=&RandomRangeVector(Minimum=<0.25 -0.1 -0.1>, Maximum=<0.5 0.1 0.1>)
     2=?
     3=?
     Index=&SpawnGroup()
     Num=::NewEnumerator0
    )
    Use Rotation Vector=true
   )
   &ShapeLocation(
    MaskSpawnGroup=true
    Override Seed=false
    Shape Origin=$Particles.SkeletalMeshLocation.SampledPosition
    Sphere Radius=5
    Sphere Surface Distribution=1
    UseOwnerScale=true
   )
   &ShapeLocation(
    Cylinder Height=20
    MaskSpawnGroup=true
    Offset Mode=::NewEnumerator0
    Override Seed=false
    Rotation Axis=<0 0 1>
    Rotation Coordinate Space=::Simulation
    Rotation Mode=::NewEnumerator1
    Shape Primitive=::NewEnumerator1
    Spawn Group Mask=1
    Surface Only=false
    UseOwnerScale=true
   ) {module=ShapeLocation001}
   &AddVelocity(
    Velocity=&MergeHelper_VectorSwitch(
     0=<0 0 1>
     1=&FN_FrontEndVector3Modifier(Gameplay=<0 0 1>, Front End=<0.5 -1 1>)
     2=?
     3=?
     Index=&SpawnGroup()
     Num=::NewEnumerator0
    )
    Velocity Speed Scale=&MergeHelper_FloatSwitch(
     0=&RandomRangeFloat(Minimum=222, Maximum=444)
     1=&RandomRangeFloat(Minimum=333, Maximum=666)
     2=?
     3=?
     Index=&SpawnGroup()
     Num=::NewEnumerator0
    )
   ) {module=AddVelocity001}
  ]
  particleUpdateScript [
   &ParticleState()
   &ScaleMeshSize(Scale Factor=&Multiply_VectorByFloat(
     Vector=&VectorFromCurve(
      VectorCurve=DataInterface() {propertyValues {XCurve {keys [<at 0 1>,<at 1 1>]}, YCurve {keys [<at 0 1>,<at 1 1>]}, ZCurve {keys [<at 0 0.1>,<at 0.4 1>]}}}
      CurveIndex=$Particles.Age
     )
     Float=&FloatFromCurve(FloatCurve=DataInterface() {propertyValues {Curve {
         defaultValue=0
         keys [
          {arriveTangent=9.0725, interpMode=RCIM_Cubic, leaveTangent=9.0725, tangentMode=RCTM_User, time=0, value=0}
          {arriveTangent=0.3041, interpMode=RCIM_Cubic, leaveTangent=0.3041, tangentMode=RCTM_User, time=0.1, value=0.9847}
          {arriveTangent=-1.0941, interpMode=RCIM_Cubic, leaveTangent=-1.0941, time=0.7473, value=0.9144}
          {arriveTangent=-0.0405, interpMode=RCIM_Cubic, leaveTangent=-0.0405, tangentMode=RCTM_User, time=1, value=0}
         ]
        }}})
    ))
   &PointForce(
    Force Falloff Distance:NiagaraFloat=&MergeHelper_FloatSwitch(0=400, 1=500, 2=?, 3=?, Index=&SpawnGroup(), Num=::NewEnumerator0)
    Force Strength=&MergeHelper_FloatSwitch(0=1000, 1=2000, 2=?, 3=?, Index=&SpawnGroup(), Num=::NewEnumerator0)
    Module.Use Falloff=true
    Module.UseLinearForce=false
   )
   &UpdateMeshOrientation(
    Rotation Rate=&RandomRangeFloat(Minimum=-1, Maximum=1)
    Rotation Vector=&Multiply_VectorByFloat(
     Vector=&RandomRangeVector(Minimum=<0 0 0>, Maximum=<1 1 1>)
     Float=&FloatFromCurve(FloatCurve=DataInterface() {propertyValues {Curve {keys [<at 0 0 RCIM_Cubic>,<at 0.1 1 RCIM_Cubic>]}}})
    )
   )
   &AerodynamicDrag(
    Aerodynamic Drag:NiagaraFloat=&MergeHelper_FloatSwitch(0=1, 1=0.25, 2=?, 3=?, Index=&SpawnGroup(), Num=::NewEnumerator0)
    Drag Shape Curve=DataInterface() {propertyValues {Curve {keys [
        {interpMode=RCIM_Cubic, tangentMode=RCTM_User, time=0, value=0.1}
        {arriveTangent=2.0799, interpMode=RCIM_Cubic, leaveTangent=2.0799, tangentMode=RCTM_User, time=1, value=1}
       ]}}}
    Ignore Mass=true
    Lift Shape Curve=DataInterface() {propertyValues {Curve {keys [
        {interpMode=RCIM_Cubic, tangentMode=RCTM_User, time=0, value=0}
        {interpMode=RCIM_Cubic, tangentMode=RCTM_User, time=0.5, value=1}
        {tangentMode=RCTM_User, time=1, value=0}
       ]}}}
    Surface Normal Axis:Vector3f=<0 0 1>
    Use Linear Drag=true
    Use Rotational Drag=true
   )
   &GravityForce(Gravity=&MergeHelper_VectorSwitch(
     0=&Add_Vector(
      A=&RandomRangeVector(Minimum=<0 0 400>, Maximum=<0 0 600>)
      B=&Lerp_Vector(
       A=<0 0 0>
       B=&FN_FrontEndVector3Modifier(Gameplay=<0 0 500>, Front End=&TransformVector(Vector=<0 -500 300>, Destination Space=::World))
       Alpha=$User.ParaGlideLeanAlpha
      )
     )
     1=&Add_Vector(
      A=&RandomRangeVector(Minimum=<0 0 600>, Maximum=<0 0 1000>)
      B=&Lerp_Vector(
       A=<0 0 0>
       B=&FN_FrontEndVector3Modifier(Gameplay=<0 0 500>, Front End=&TransformVector(Vector=<250 -1000 500>, Destination Space=::World))
       Alpha=$User.ParaGlideLeanAlpha
      )
     )
     2=?
     3=?
     Index=&SpawnGroup()
     Num=::NewEnumerator0
    ))
   &CurlNoiseForce(
    Bias Noise=false
    Noise Frequency=0.03
    Noise Quality / Cost=::NewEnumerator0
    Noise Strength=&RandomRangeFloat(Minimum=200, Maximum=500)
    Pan Noise=false
   )
   &SolveForcesAndVelocity(Clamp Velocity=false, Limit Acceleration=false)
   &AlignSpriteToMeshOrientation(SetAlignment=true, SetFacing=true)
   &self:SphereBlockingVolume(
    Sphere Origin:Vector3f=&ReturnCameraPosition(Coordinate Space:ENiagaraCoordinateSpace=::Simulation, Camera Query:NiagaraDataInterfaceCamera=DataInterface())
    Sphere Radius=66
   ) {module=SphereBlockingVolume002}
  ]
  rendererClasses [&NiagaraMeshRendererProperties]
  renderers [{rendererClass=&NiagaraMeshRendererProperties, rendererIndex=0}]
 }
 Ducks {
  emitterUpdateScript [
   &EmitterState(
    Inactive Response=::NewEnumerator1
    Life Cycle Mode=::NewEnumerator1
    Loop Behavior=::NewEnumerator1
    Loop Duration Mode=::NewEnumerator1
    UseLoopDelay=true
   )
   &SpawnRate(SpawnRate=&Lerp_Float(A=5, B=10, Alpha=$User.ParaGlideLeanAlpha), Use Spawn Probability=true)
   &SpawnRate(SpawnGroup=1, SpawnRate=&Lerp_Float(A=0, B=4, Alpha=$User.ParaGlideLeanAlpha), Use Spawn Probability=false) {module=SpawnRate001}
  ]
  particleSpawnScript [
   &PartitionParticles()
   &InitializeParticle(
    Lifetime=&RandomRangeFloat(Minimum=1, Maximum=1.5)
    Mesh Scale Mode=::NewEnumerator1
    Mesh Uniform Scale Max=0.03
    Mesh Uniform Scale Min=0.02
    Sprite Rotation Mode=::NewEnumerator2
    UsePositionOffset=true
    Write Material Random=true
   )
   &SkeletalMeshLocation(
    Bone / Socket Index=$Particles.PartitionParticles.Partition
    Bone Sampling Mode=::NewEnumerator4
    MaskSpawnGroup=true
    Orientation Sampling=::NewEnumerator1
    Position Sampling=::NewEnumerator1
    Sampled Position Offset=&FN_FrontEndVector3Modifier(Gameplay=<0 -5 0>, Front End=<0 0 10>)
    Sampled Position Offset Coordinate Space:ENiagaraSkelMeshTransforms=::NewEnumerator1
    Skeletal Mesh=$System.Skeletal Mesh
   )
   &SkeletalMeshLocation(
    Bone / Socket Index=4
    Bone Sampling Mode=::NewEnumerator4
    MaskSpawnGroup=true
    Orientation Sampling=::NewEnumerator1
    Position Sampling=::NewEnumerator1
    Sampled Position Offset=&FN_FrontEndVector3Modifier(Gameplay=<0 -85 -200>, Front End=<0 85 -200>)
    Sampled Position Offset Coordinate Space:ENiagaraSkelMeshTransforms=::NewEnumerator2
    Skeletal Mesh=$System.Skeletal Mesh
    Spawn Group Mask=1
   ) {module=SkeletalMeshLocation001}
   &ShapeLocation(
    Cylinder Height=20
    Cylinder Radius=50
    MaskSpawnGroup=true
    Offset Mode=::NewEnumerator0
    Override Seed=false
    Rotation Axis=<0 0 1>
    Rotation Coordinate Space=::Simulation
    Rotation Mode=::NewEnumerator1
    Shape Primitive=::NewEnumerator1
    Spawn Group Mask=1
    Surface Only=false
    UseOwnerScale=true
   )
   &AddVelocity(
    Velocity=&Add_Vector(
     A=&RandomRangeVector(Minimum=<-5 -5 0>, Maximum=<5 5 20>)
     B=&Lerp_Vector(
      A=<0 -12 0>
      B=<0 -15 0>
      Alpha=&FN_FrontEndFloatModifier(Gameplay=0, Front End=$User.ParaGlideLeanAlpha)
     )
    )
    Velocity Speed Scale=&RandomRangeFloat(Minimum=4, Maximum=12)
   ) {module=AddVelocity001}
   &InitialMeshOrientation(Use Rotation Vector=false)
   &DynamicMaterialParameters(
    Index 0 Param 1=&RandomRangeFloat(Minimum=0, Maximum=1)
    Index 0 Param 2=1
    Module.Index0Float0Write=true
    Module.Index0Float1Write=true
    Module.Index0Float2Write=true
    Module.Index0Float3Write=true
   ) {module=DynamicMaterialParameters001}
  ]
  particleUpdateScript [
   &ParticleState()
   &PointForce(
    Force Falloff Distance=400
    Force Strength=&MergeHelper_FloatSwitch(0=0, 1=300, 2=?, 3=?, Index=&SpawnGroup(), Num=::NewEnumerator0)
    Module.Use Falloff=true
    Module.UseLinearForce=false
   )
   &ScaleMeshSize(Scale Factor=&VectorFromFloat(Value=&FloatFromCurve(
      FloatCurve=DataInterface() {propertyValues {Curve {keys [<at 0 0>,<at 0.2 1>,<at 0.2 1>,<at 0.8 1>,<at 1 0>]}}}
      Scale Curve=/Niagara/DynamicInputs/Camera/CameraFade(
       Camera Query:NiagaraDataInterfaceCamera=DataInterface() {propertyValues {bRequireCurrentFrameData=false}}
       Fade Offset:NiagaraFloat=50
       Fade Distance:NiagaraFloat=150
       Use Owner Position:NiagaraBool=true
      )
     )))
   &UpdateMeshOrientation(Rotation Rate=&RandomRangeFloat(Minimum=0.2, Maximum=0.7))
   &ScaleColor(
    Scale Alpha=&Multiply_Float(
     A=&RandomRangeFloat(Minimum=0.7, Maximum=1)
     B=&FloatFromCurve(
      FloatCurve=DataInterface() {propertyValues {Curve {keys [
          {arriveTangent=3.7485, interpMode=RCIM_Cubic, leaveTangent=3.7485, tangentMode=RCTM_User, time=0, value=0.35}
          {interpMode=RCIM_Cubic, tangentMode=RCTM_User, time=0.5, value=1}
          {tangentMode=RCTM_User, time=1, value=0}
         ]}}}
      Scale Curve=2
     )
    )
    Scale RGB=&MakeVectorFromLinearColorRGB(Color=&ColorFromCurve(
      ColorCurve=DataInterface() {propertyValues {
        AlphaCurve {keys [<at 0 1 RCIM_Cubic>]}
        BlueCurve {keys [<at 0.0515 0.3482 RCIM_Cubic>,<at 0.785 0.0273 RCIM_Cubic>]}
        GreenCurve {keys [<at 0.0515 0.1808 RCIM_Cubic>,<at 0.785 0.011 RCIM_Cubic>]}
        RedCurve {keys [<at 0.0515 1 RCIM_Cubic>,<at 0.785 0.0116 RCIM_Cubic>]}
       }}
      CurveIndex=&Add_Float(
       A=$Particles.Age
       B=&Lerp_Float(
        A=0
        B=0.3
        Alpha=&FloatFromCurve(FloatCurve=DataInterface() {propertyValues {Curve {keys [<at 0.7 0>,<at 1 1>]}}}, CurveIndex=$User.ParaGlideLeanAlpha)
       )
      )
     ))
    ScaleA=true
    ScaleRGB=true
   )
   &Drag(
    Drag=&MergeHelper_FloatSwitch(
     0=&RandomRangeFloat(Minimum=2, Maximum=3)
     1=&RandomRangeFloat(Minimum=0.5, Maximum=1)
     2=?
     3=?
     Index=$Particles.VisibilityTag
     Num=::NewEnumerator0
    )
    Use Linear Drag=true
    Use Rotational Drag=true
   )
   &GravityForce(Gravity=&Add_Vector(
     A=&RandomRangeVector(Minimum=<0 0 300>, Maximum=<0 0 500>)
     B=&Lerp_Vector(
      A=<0 0 0>
      B=&FN_FrontEndVector3Modifier(Gameplay=<0 0 500>, Front End=&TransformVector(Vector=<0 -500 300>, Destination Space=::World))
      Alpha=$User.ParaGlideLeanAlpha
     )
    ))
   &CurlNoiseForce(
    Bias Noise=false
    Noise Frequency=55
    Noise Strength=&RandomRangeFloat(Minimum=100, Maximum=200)
    Pan Noise=false
   )
   &SolveForcesAndVelocity(Clamp Velocity=false, Limit Acceleration=false)
  ]
  rendererClasses [&NiagaraMeshRendererProperties]
  renderers [{rendererClass=&NiagaraMeshRendererProperties, rendererIndex=0}]
 }
 Speedlines {
  emitterUpdateScript [
   &EmitterState()
   &SpawnRate(
    SpawnRate=&MakeCustomFloatFromBool(Bool=$User.bIsFrontEndPreview, True Float=45, False Float=70)
    Use Spawn Probability=true
   ) {module=SpawnRate001}
  ]
  particleSpawnScript [
   &InitializeParticle(
    Color=&MakeLinearColorFromVectorAndFloat(
     "Vector (RGB)"=<1 1 1>
     "Float (Alpha)"=&Multiply_Float(
      A=&UniformRangedFloat(Minimum=1, Maximum=1)
      B=&MakeCustomFloatFromBool(
       Bool=$User.bIsFrontEndPreview
       True Float=&FloatFromCurve(
        FloatCurve=DataInterface() {propertyValues {Curve {keys [{tangentMode=RCTM_User, time=0, value=0}, {tangentMode=RCTM_User, time=2, value=1}]}}}
        CurveIndex=$Emitter.Age
        Scale Curve=0.4
       )
       False Float=1
      )
     )
    )
    Lifetime Max=0.5
    Lifetime Min=0.3
    Lifetime Mode=::NewEnumerator1
    Position Offset=&FN_FrontEndVector3Modifier(Gameplay=<0 -50 0>, Front End=<0 50 0>)
    Sprite Size Max=<4 10>
    Sprite Size Min=<3 2>
    Sprite Size Mode=::NewEnumerator2
    UsePositionOffset=true
    Write Material Random=true
   )
   &AddVelocity(
    Module.Use Falloff Distance=false
    Origin Offset=&FN_FrontEndVector3Modifier(Gameplay=<0 0 1000>, Front End=<0 -250 250>)
    Override Seed=false
    Velocity Mode=::NewEnumerator1
    Velocity Speed=-2500
   ) {module=AddVelocity001}
   &TorusLocation(
    Align Torus With Axis=true
    Coordinate Space:ENiagaraCoordinateSpace=::Simulation
    Handle Radius=0
    Large Radius:NiagaraFloat=&FN_FrontEndFloatModifier(
     Gameplay=&RandomRangeFloat(Minimum=350, Maximum=550)
     Front End=&RandomRangeFloat(Minimum=350, Maximum=450)
    )
    MaskSpawnGroup=false
    Module.Override Randomness=false
    Offset=<0 33 0>
    Torus Origin:NiagaraPosition=&AddVectorToPosition(
     Position=$Particles.Position
     Vector=&FN_FrontEndVector3Modifier(Gameplay=&RandomRangeVector(Minimum=<0 0 -250>, Maximum=<0 0 -750>), Front End=<0 500 -500>)
    )
   )
   &AddVelocity(
    Velocity=&Multiply_VectorByFloat(
     Vector=&FN_FrontEndVector3Modifier(Gameplay=$System.SkydiveVelocity, Front End=<0 1 1>)
     Float=-0.1
    )
    Velocity Speed Scale=&FN_FrontEndFloatModifier(Gameplay=1, Front End=1)
   )
  ]
  particleUpdateScript [
   &SolveForcesAndVelocity(Clamp Velocity=false, Limit Acceleration=false)
   &ParticleState()
   &ScaleSpriteSize(Scale Factor=&Vector2DFromCurve(
     Vector2Curve=DataInterface() {propertyValues {XCurve {keys [<at 0 0.5 RCIM_Cubic>,<at 0.5 1 RCIM_Cubic>]}, YCurve {keys [<at 0 0 RCIM_Cubic>,<at 0.5 1 RCIM_Cubic>]}}}
     ScaleCurve=&Vector2DFromFloat(Value=1)
    ))
   &ScaleSpriteSizeBySpeed(Max Scale Factor=<1 50>, Min Scale Factor=<0.01 0.05>, Velocity Threshold=1)
   &ScaleColor(
    Scale Alpha=&FloatFromCurve(
     FloatCurve=DataInterface() {propertyValues {Curve {keys [
         {interpMode=RCIM_Cubic, tangentMode=RCTM_User, time=0, value=0}
         {interpMode=RCIM_Cubic, tangentMode=RCTM_User, time=0.4, value=1}
         {tangentMode=RCTM_User, time=1, value=0}
        ]}}}
     Scale Curve=&Lerp_Float(A=0, B=0.15, Alpha=$System.Dive)
    )
    Scale RGB=&VectorFromFloat(Value=&FN_FrontEndFloatModifier(Gameplay=0.5, Front End=0.4))
    ScaleA=true
    ScaleRGB=true
   )
  ]
  rendererClasses [&NiagaraSpriteRendererProperties]
  renderers [{rendererClass=&NiagaraSpriteRendererProperties}]
 }
 WindTrails {
  emitterUpdateScript [
   &EmitterState()
   &SpawnBurst_Instantaneous(Spawn Count=40, Use Spawn Probability=false, UseLoopCountLimit=true)
  ]
  particleSpawnScript [
   &PartitionParticles(
    Normalized Index In Partition Includes One=true
    Number Of Partitions=4
    Particles Per Partition=10
    Partition Mode:ENiagara_PartitionMode_v2=::NewEnumerator1
    Write Partition Index to Ribbon Link Order:ENiagara_AttributeSamplingApplyOutput=::NewEnumerator0
    Write Partition to Ribbon ID=::NewEnumerator0
   )
   &InitializeParticle(
    Color=<rgba 1 1 1 0.2>
    Lifetime=&MergeHelper_FloatSwitch(0=0.25, 1=0.25, 2=0.15, 3=0.15, Index=$Particles.PartitionParticles.Partition, Num=::NewEnumerator2)
    Ribbon Width=&SelectFloatFromArray(Direct Array Index=$Particles.PartitionParticles.Partition, Float Selection Array=DataInterface() {propertyValues {FloatData [30,30,20,20]}})
    Ribbon Width Mode=::NewEnumerator1
    Sprite Size Mode=::NewEnumerator3
    UsePositionOffset=false
    Write Material Random=true
   )
  ]
  particleUpdateScript [
   &ParticleState(Kill Particles When Lifetime Has Elapsed=false)
   &SkeletalMeshLocation(
    Bone / Socket Index=$Particles.PartitionParticles.Partition
    Bone Sampling Mode=::NewEnumerator4
    MaskSpawnGroup=false
    Skeletal Mesh=$System.Skeletal Mesh
   )
   /SetVariables(Particles.NewAge:NiagaraFloat=&Multiply_Float(A=$Particles.PartitionParticles.Normalized Index In Partition, B=$Particles.Initial.Lifetime)) {isSetParameters=true}
   /SetVariables(Particles.Velocity=&Multiply_VectorByFloat(
     Vector=&Lerp_Vector(A=$System.FN_Contrail_Setup.Direction Velocity Raw, B=$System.FN_Contrail_Setup.Direction Velocity, Alpha=$Particles.NewAge)
     Float=$Particles.Initial.Lifetime
    )) {isSetParameters=true}
   /SetVariables(Particles.Position=&AddVectorToPosition(
     Position=$Particles.Position
     Vector=&Lerp_Vector(
      A=<0 0 0>
      B=&Multiply_VectorByFloat(Vector=$Particles.Velocity, Float=$Particles.PartitionParticles.Normalized Index In Partition)
      Alpha=&FloatFromCurve(
       FloatCurve=DataInterface() {propertyValues {Curve {keys [<at 0 0 RCIM_Cubic>, {interpMode=RCIM_Cubic, leaveTangent=2.2081, time=1, value=1}]}}}
       CurveIndex=$Emitter.Age
      )
     )
    )) {isSetParameters=true}
   &ScaleColor(
    Scale Alpha=&Multiply_Float(
     A=&Multiply_Float(
      A=&FN_FrontEndFloatModifier(Gameplay=0.4, Front End=0.3)
      B=$System.FN_Contrail_Setup.Normalized Intro Age
     )
     B=&FloatFromCurve(
      FloatCurve=DataInterface() {propertyValues {Curve {keys [<at 0 0.5>,<at 0.2 1>,<at 1 0>]}}}
      CurveIndex=$Particles.PartitionParticles.Normalized Index In Partition
     )
    )
    Scale RGB=&MakeVectorFromLinearColorRGB(Color=$System.ToD Color)
    ScaleA=true
    ScaleRGB=true
   )
   /SetVariables(Particles.FillTrailDynPar:Vector4f=&MakeVector4(
     X=&Add_Float(
      A=&RandomRangeFloat(Minimum=0, Maximum=100, Randomness Mode=::NewEnumerator1, Random Seed=$Particles.PartitionParticles.Partition, Fixed Random Seed=true)
      B=&Multiply_Float(A=$Emitter.Age, B=1.5)
     )
     Y=0
     Z=0
     W=1
    )) {isSetParameters=true}
   &DynamicMaterialParameters(Module.Index0Float0Write=true, Module.Index0Float1Write=true, Module.Index0Float2Write=true, Module.Index0Float3Write=true)
   &ScaleRibbonWidth(Ribbon Width Scale=&FloatFromCurve(
     FloatCurve=DataInterface() {propertyValues {Curve {keys [<at 0 0.3>,<at 0.2 1>,<at 1 1>]}}}
     CurveIndex=$Particles.PartitionParticles.Normalized Index In Partition
    ))
  ]
  rendererClasses [&NiagaraRibbonRendererProperties]
  renderers [
   {rendererClass=&NiagaraRibbonRendererProperties}
   {rendererClass=&NiagaraRibbonRendererProperties, rendererIndex=1}
  ]
 }
}
summary {
 systemName=NS_Contrail_FunkyFeather
 emitters [
  {emitterName=Bills, rendererClasses [&NiagaraMeshRendererProperties]}
  {emitterName=Ducks, rendererClasses [&NiagaraMeshRendererProperties]}
  {emitterName=Speedlines, rendererClasses [&NiagaraSpriteRendererProperties]}
  {emitterName=WindTrails, rendererClasses [&NiagaraRibbonRendererProperties]}
 ]
 userVariables [
  User.bIsFrontEnd {type=NiagaraBool}
  User.bIsFrontEndPreview {type=NiagaraBool}
  User.ParaGlideLeanAlpha {type=NiagaraFloat, defaultValue=0, description=""}
 ]
}
